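# Additive unlit material that draws an explosion wind vector as signed emissive color, for particle sprites or meshes.
# The normalized direction to the particle or pivot center is crossed with a down vector, giving a tangential XY vector scaled by intensity with blue zeroed.
# A squared SphereMask falloff forms a soft ring, faded by particle alpha or an Opacity parameter.
# Wind helper WPO positions vertices, adding projection to an Explosion Radius around the pivot when particles are off.
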Material /Game/WaterAndWind/Materials/Forces/ExplosionWindVector
graph {
 properties {
  bAllowNegativeEmissiveColor=true
  bUsedWithParticleSprites=true
  blendMode=BLEND_Additive
  pixelDepthOffsetMode=PDOM_Legacy
  shadingModel=MSM_Unlit
  usedShadingModels=Unlit
 }
 nodes [
  n1 = StaticSwitchParameter(
   True=ParticlePositionWS()
   False=MaterialFunctionCall() {materialFunction=&ComponentPivotLocation}.Result
  ) {defaultValue=true, parameterName=Use with Particles}
  n2 = MaterialFunctionCall!(World position) {materialFunction=/Game/WaterAndWind/MaterialFunctions/Fortnite_WindVectorParticleHelperFunctions}
  n3 = CrossProduct(
   A=Constant3Vector() {constant=<rgba 0 0 -1 0>}
   B=MaterialFunctionCall(Vector=Subtract(A=@n1, B=@n2.Pixel Shader Particle World Position - Adjusted For Offset)) {materialFunction=&SafeNormalize}.Result
  )
  n4 = ScalarParameter() {defaultValue=1, parameterName=Intensity Tweak}
  MaterialFunctionCall(Emissive=Multiply(A=Multiply(A=@n3, B=@n4), B=Constant3Vector() {constant=<rgba 1 1 0 0>}) {desc=Mask Blue Channel Information}) {materialFunction=&MF_FX_EmissiveAdjustments}
  n5 = SphereMask(A=TextureCoordinate(), B=Constant() {r=0.5}) {attenuationRadius=0.5, hardnessPercent=0}
  n6 = OneMinus(None=@n5)
  n7 = Multiply(A=Multiply(A=Multiply(A=@n6, B=@n6), B=@n5)) {constB=6}
  n8 = ParticleColor()
  n9 = ScalarParameter() {defaultValue=1, parameterName=Opacity}
  Reroute(None=Multiply(A=Multiply(A=@n7, B=@n7), B=StaticSwitchParameter(True=@n8.A, False=@n9) {defaultValue=true, parameterName=Use with Particles}))
  n10 = Constant()
  n11 = WorldPosition()
  n12 = MaterialFunctionCall(
   Vector=Subtract(
    A=@n11.XYZ
    B=MaterialFunctionCall() {materialFunction=&ComponentPivotLocation}.Result
   ) {desc=Outward Vector}
   Default=Constant3Vector()
  ) {materialFunction=&SafeNormalize}
  n13 = ScalarParameter() {defaultValue=500, parameterName=Explosion Radius}
  n14 = Add(
   A=Subtract(A=Multiply(A=@n12.Result, B=@n13), B=@n11.XYZ)
   B=MaterialFunctionCall() {materialFunction=&ComponentPivotLocation}.Result
  )
  Reroute(None=Add(
    A=MaterialFunctionCall() {materialFunction=&Fortnite_WindVectorParticleHelperFunctions}.Wind Particle WPO
    B=StaticSwitchParameter(True=@n10, False=@n14) {defaultValue=true, parameterName=Use with Particles}
   ))
 ]
}
