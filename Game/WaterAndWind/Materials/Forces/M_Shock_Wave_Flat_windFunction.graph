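# Translucent, unlit, two-sided mesh-particle material for a flat shock wave in the wind force effects.
# It normalizes the difference between the wind helper function's particle WPO and the particle position, then feeds that direction as color through an emissive adjustment function.
# A mask multiplies a smoke ramp, a soft vertical ramp and a squared side-edge falloff in U.
# Dynamic parameters offset the smoke ramp in V (R) and scale the mask (G).
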Material /Game/WaterAndWind/Materials/Forces/M_Shock_Wave_Flat_windFunction
graph {
 properties {
  bUsedWithMeshParticles=true
  blendMode=BLEND_Translucent
  pixelDepthOffsetMode=PDOM_Legacy
  shadingModel=MSM_Unlit
  twoSided=true
  usedShadingModels=Unlit
 }
 nodes [
  n1 = MaterialFunctionCall!(World position) {materialFunction=/Game/WaterAndWind/MaterialFunctions/Fortnite_WindVectorParticleHelperFunctions}
  n2 = ParticlePositionWS()
  MaterialFunctionCall(Emissive=Normalize(VectorInput=Subtract(A=@n1.Wind Particle WPO, B=@n2))) {materialFunction=&MF_FX_EmissiveAdjustments}
  MaterialFunctionCall() {materialFunction=&Fortnite_WindVectorParticleHelperFunctions}
  n3 = DynamicParameter() {defaultValue=<rgba 0 0 0 1>}
  n4 = Add(A=TextureCoordinate(), B=AppendVector(A=Constant(), B=@n3.R))
  n5 = TextureSample() {texture=/Game/Global/Textures/Default/Masks_and_Gradients/T_Soft_Ramp_Vert_03}
  n6 = TextureCoordinate()
  n7 = Abs(None=ConstantBiasScale(None=MaterialFunctionCall(Float2=@n6) {materialFunction=&BreakOutFloat2Components}.R) {bias=-0.5, scale=2})
  Multiply(
   A=Multiply(A=TextureSample(UVs=@n4) {texture=/Game/Effects/Fort_Effects/Textures/Ramps/T_Smoke_Shoot_Out_Ramp}.G, B=Multiply(A=@n5.G, B=@n3.G))
   B=OneMinus(None=Multiply(A=@n7, B=@n7))
  )
 ]
}
