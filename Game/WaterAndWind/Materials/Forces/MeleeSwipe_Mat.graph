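# Additive unlit mesh-particle material that renders a melee swipe as a directional force for the water and wind system, with negative emissive allowed for signed vectors.
# Emissive is the normalized direction from particle center to pixel, flipped by a left/right parameter and perturbed by panning wind turbulence.
# A squared mask combines a smooth ramp, sine and UV edge falloff, offset by particle red for the sweep and scaled by particle alpha.
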Material /Game/WaterAndWind/Materials/Forces/MeleeSwipe_Mat
graph {
 properties {
  bAllowNegativeEmissiveColor=true
  bUsedWithMeshParticles=true
  blendMode=BLEND_Additive
  pixelDepthOffsetMode=PDOM_Legacy
  shadingModel=MSM_Unlit
  usedShadingModels=Unlit
 }
 nodes [
  MaterialFunctionCall!(World position) {materialFunction=/Game/WaterAndWind/MaterialFunctions/Fortnite_WindVectorParticleHelperFunctions}
  n1 = Normalize(VectorInput=Subtract(
    A=MaterialFunctionCall() {materialFunction=&Fortnite_WindVectorParticleHelperFunctions}.Pixel Shader Particle World Position - Adjusted For Offset
    B=ParticlePositionWS()
   ))
  n2 = Multiply(A=@n1) {constB=-1}
  n3 = ScalarParameter() {parameterName="Swipe Left (0) OR Right (1)"}
  n4 = TextureSample(UVs=Panner(Coordinate=TextureCoordinate() {coordinateIndex=1, vTiling=0.6}) {bFractionalPart=true, speedX=0.5, speedY=1}) {
   samplerType=SAMPLERTYPE_LinearColor
   texture=/Engine/Functions/Engine_MaterialFunctions02/ExampleContent/Textures/WindTurblenceVectorAndGustMagnitude
  }
  MaterialFunctionCall(Emissive=Add(A=LinearInterpolate(A=@n1, B=@n2, Alpha=@n3), B=ConstantBiasScale(None=@n4.RGB) {bias=-0.5, scale=0.1})) {materialFunction=&MF_FX_EmissiveAdjustments}
  n5 = MaterialFunctionCall(Float2=TextureCoordinate() {coordinateIndex=1, vTiling=0.6}) {materialFunction=&BreakOutFloat2Components}
  n6 = OneMinus(None=@n5.R)
  n7 = Add(A=Add(A=LinearInterpolate(A=@n6, B=@n5.R, Alpha=@n3), B=@n5.G), B=ParticleColor().R)
  n8 = AppendVector(A=@n5.R, B=@n7)
  n9 = Subtract(A=TextureCoordinate()) {constB=0.5}
  n10 = Abs(None=Add(A=@n9, B=@n9))
  n11 = MaterialFunctionCall(Float2=OneMinus(None=Multiply(A=@n10, B=@n10))) {materialFunction=&BreakOutFloat2Components}
  n12 = Multiply(
   A=Multiply(A=TextureSample(UVs=@n8) {texture=/Game/Effects/Fort_Effects/Textures/Ramps/T_Ramp_Smooth}.R, B=Sine(None=@n7))
   B=Multiply(A=@n11.R, B=@n11.R)
  )
  Multiply(A=Multiply(A=@n12, B=@n12), B=ParticleColor().A)
 ]
}
