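# Additive unlit particle material that renders a player push force as signed vector colors, with negative emissive allowed.
# A normalized direction from each particle's center, rotated about Z, is shaped by a soft radial falloff and modulated by panning world-space noise.
# A Bounce switch adds random noise timing and lifetime sine pulsing; particle alpha scales the result before emissive adjustment.
# A height-based occlusion chain is left unconnected.
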Material /Game/WaterAndWind/Materials/Forces/PlayerPush_Particles
graph {
 properties {
  bAllowNegativeEmissiveColor=true
  bUsedWithParticleSprites=true
  blendMode=BLEND_Additive
  pixelDepthOffsetMode=PDOM_Legacy
  shadingModel=MSM_Unlit
  usedShadingModels=Unlit
 }
 nodes [
  n1 = ParticleRelativeTime()
  n2 = ComponentMask(None=Divide(
    A=MaterialFunctionCall!(World position) {materialFunction=/Game/WaterAndWind/MaterialFunctions/Fortnite_WindVectorParticleHelperFunctions}.Pixel Shader Particle World Position - Adjusted For Offset
    B=Constant() {r=200}
   )) {g=true, r=true}
  n3 = Add(A=StaticSwitchParameter(True=ParticleRandom(), False=Constant()) {parameterName=Use Bounce animation}, B=Time())
  n4 = Divide(
   A=Subtract(
    A=ParticlePositionWS()
    B=MaterialFunctionCall() {materialFunction=&Fortnite_WindVectorParticleHelperFunctions}.Pixel Shader Particle World Position - Adjusted For Offset
   )
   B=ComponentMask(None=ParticleSize()) {r=true}
  )
  n5 = Constant3Vector() {constant=<rgba 1 0 0 0>}
  n6 = MaterialFunctionCall(Vector=@n4, Default=Normalize(VectorInput=Add(A=@n4, B=@n5))) {materialFunction=&SafeNormalize}
  n7 = RotateAboutAxis(
   NormalizedRotationAxis=Constant3Vector() {constant=<rgba 0 0 1 0>}
   RotationAngle=ScalarParameter() {defaultValue=0.7304, parameterName=Rotate}
   PivotPoint=Constant3Vector()
   Position=@n6.Result
  )
  n8 = SphereMask(A=@n4, B=Constant()) {attenuationRadius=0.75, hardnessPercent=0}
  n9 = ScalarParameter() {defaultValue=8.5313, parameterName=Power}
  n10 = Multiply(
   A=TextureSample(UVs=Panner(Coordinate=@n2, Time=@n3) {speedX=0.1, speedY=0.1}) {samplerType=SAMPLERTYPE_Grayscale, texture=&LowResBlurredNoise}.R
   B=Multiply(A=Add(A=@n7, B=@n6.Result), B=Power(Base=@n8, Exp=@n9))
  )
  n11 = Multiply(A=ConstantBiasScale(None=Sine(None=@n1)), B=@n10)
  MaterialFunctionCall(Emissive=Multiply(A=StaticSwitchParameter(True=@n11, False=@n10) {desc=Emissive, parameterName=Use Bounce animation}, B=ParticleColor().A)) {materialFunction=&MF_FX_EmissiveAdjustments}
  MaterialFunctionCall() {materialFunction=&Fortnite_WindVectorParticleHelperFunctions}
  n12 = WorldPosition()
  n13 = Subtract(A=Divide(A=ComponentMask(None=@n12.XYZ) {g=true, r=true}, B=Constant() {desc=radius, r=10240})) {constB=0.5}
  n14 = ComponentMask(None=@n12.XYZ) {b=true}
  n15 = OneMinus(None=Subtract(A=TextureSample(UVs=@n13).A) {constB=15360})
  n16 = Constant()
  n17 = Add(
   A=Multiply(A=TextureSample(UVs=@n13).B) {constB=3}
   B=If(A=@n14, B=@n15, "A > B"=@n16, "A == B"=@n16, "A < B"=Constant() {r=1}) {equalsThreshold=10}
  ) {desc=White occlusion}
  OneMinus(None=Multiply(A=Clamp(None=@n17)) {constB=0.9})
  SphereMask(A=@n15, B=@n14)
 ]
}
