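# WaveMaterial is an additive unlit particle material that appears to draw water wave forces into an offscreen wind-and-water render target.
# Three chained flow map samples advect particles downstream, scaled by particle red, and a helper offset moves them out of the main view.
# A splash height map, flow-distorted by water flow plus wind turbulence, blends two channels by flow strength into emissive blue, scaled by particle alpha and Normal Intensity.
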Material /Game/WaterAndWind/Materials/Forces/WaveMaterial
graph {
 properties {
  bAllowNegativeEmissiveColor=true
  bUsedWithNanite=true
  bUsedWithParticleSprites=true
  blendMode=BLEND_Additive
  pixelDepthOffsetMode=PDOM_Legacy
  shadingModel=MSM_Unlit
  translucencyPass=MTP_BeforeDOF
  usedShadingModels=Unlit
 }
 nodes [
  n1 = ParticlePositionWS()
  n2 = MaterialFunctionCall!(Position=@n1) {materialFunction=/Game/WaterAndWind/MaterialFunctions/Fortnite_SampleWindWaterIsometricWorldData}
  n3 = Constant()
  n4 = Multiply(
   A=Multiply(A=AppendVector(A=@n2.Flow Map RG, B=@n3), B=ScalarParameter() {defaultValue=65.4642, parameterName=Water Speed Factor})
   B=ParticleColor().R
  ) {constB=2}
  n5 = MaterialFunctionCall(Position=Add(A=@n4, B=@n1)) {materialFunction=&Fortnite_SampleWindWaterIsometricWorldData}
  n6 = Constant()
  n7 = Multiply(
   A=Multiply(A=AppendVector(A=@n5.Flow Map RG, B=@n6), B=ScalarParameter() {defaultValue=65.4642, parameterName=Water Speed Factor})
   B=ParticleColor().R
  ) {constB=2}
  n8 = Add(A=@n7, B=@n1)
  n9 = Multiply(
   A=AppendVector(
    A=MaterialFunctionCall(Position=@n8) {materialFunction=&Fortnite_SampleWindWaterIsometricWorldData}.Flow Map RG
    B=Constant()
   )
   B=ScalarParameter() {defaultValue=65.4642, parameterName=Water Speed Factor}
  )
  n10 = ParticleColor()
  Add(
   A=Add(A=Add(A=@n4, B=@n7), B=Multiply(A=@n9, B=@n10.R) {constB=2}) {desc=full particle offset}
   B=MaterialFunctionCall!(World position) {materialFunction=/Game/WaterAndWind/MaterialFunctions/Fortnite_WindVectorParticleHelperFunctions}.Wind Particle WPO
  ) {constB=0, desc=move out of view of the main render target}
  n11 = TextureObject() {
   samplerType=SAMPLERTYPE_LinearColor
   texture=/Engine/Functions/Engine_MaterialFunctions02/ExampleContent/Textures/WindTurblenceVectorAndGustMagnitude
  }
  n12 = MaterialFunctionCall() {materialFunction=&Fortnite_WindVectorParticleHelperFunctions}
  n13 = MaterialFunctionCall(TextureObject=@n11, TextureSize=Constant() {r=256}, WorldPosition=@n12.Pixel Shader Particle World Position - Adjusted For Offset) {materialFunction=&WorldAlignedTexture}
  n14 = MaterialFunctionCall(Position=@n12.Pixel Shader Particle World Position - Adjusted For Offset) {materialFunction=&Fortnite_SampleWindWaterIsometricWorldData}
  n15 = Add(A=ComponentMask(None=ConstantBiasScale(None=@n13.Z Texture) {bias=-0.5, scale=2}) {g=true, r=true}, B=@n14.Flow Map RG)
  n16 = MaterialFunctionCall(Float3=MaterialFunctionCall!(
    Diffuse=TextureObject() {samplerType=SAMPLERTYPE_LinearColor, texture=/Game/WaterAndWind/Textures/SplashHeightMap}
    Normal Map
    "Flow Vector Map (see tooltip)"=Multiply(A=@n15) {constB=0.1}
    UV texture 2 offset=Constant()
    "Time (see tooltip)"
    UVs=TextureCoordinate()
   ) {materialFunction=/Engine/Functions/Engine_MaterialFunctions02/Texturing/FlowMaps_Simple}.Diffuse) {materialFunction=&BreakOutFloat3Components}
  n17 = MaterialFunctionCall(Position=ParticlePositionWS()) {materialFunction=&Fortnite_SampleWindWaterIsometricWorldData}
  n18 = Constant()
  n19 = AppendVector(
   A=Constant2Vector()
   B=LinearInterpolate(A=Multiply(A=@n16.R) {constB=2}, B=@n16.G, Alpha=SphereMask(A=@n17.Flow Map RG, B=@n18) {attenuationRadius=0.5, hardnessPercent=0})
  )
  n20 = ParticleColor()
  MaterialFunctionCall(Emissive=Multiply(A=Multiply(A=@n19, B=@n20.A), B=ScalarParameter() {defaultValue=2.9377, parameterName=Normal Intensity})) {materialFunction=&MF_FX_EmissiveAdjustments}
  MaterialFunctionCall() {materialFunction=&Fortnite_WindVectorParticleHelperFunctions}
  n21 = ConstantBiasScale(None=TextureCoordinate()) {bias=-0.5, scale=2}
  MaterialFunctionCall!(f2=OneMinus(None=Abs(None=@n21)), f3, f4) {materialFunction=/Game/Packages/Fortnite_UniqueMaterials_1/VectorComponentOps}
  MaterialFunctionCall() {materialFunction=&Fortnite_SampleWindWaterIsometricWorldData}
  Multiply(A=@n14.Water Z height)
  SphereMask(
   A=MaterialFunctionCall(Float3=MaterialFunctionCall() {materialFunction=&Fortnite_WindVectorParticleHelperFunctions}.Pixel Shader Particle World Position - Adjusted For Offset) {materialFunction=&BreakOutFloat3Components}.B
   B=@n14.Water Z height
  ) {attenuationRadius=512, hardnessPercent=50}
 ]
}
