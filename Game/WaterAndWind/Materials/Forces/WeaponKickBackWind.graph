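# Additive unlit particle material that, going by its name and helper functions, renders weapon kickback as a wind force vector.
# For mostly horizontal particles, WPO rebuilds the sprite along its travel direction and uses a weapon-fire mask texture; steep ones use a radial mask.
# Emissive encodes a signed tangential vector around the particle center, pulsed by a sine of lifetime plus mask.
# Opacity combines mask, particle and vertex alpha, fading with upward direction.
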Material /Game/WaterAndWind/Materials/Forces/WeaponKickBackWind
graph {
 properties {
  bAllowNegativeEmissiveColor=true
  bUsedWithNanite=true
  bUsedWithParticleSprites=true
  blendMode=BLEND_Additive
  pixelDepthOffsetMode=PDOM_Legacy
  shadingModel=MSM_Unlit
  usedShadingModels=Unlit
 }
 nodes [
  n1 = DotProduct(
   A=MaterialFunctionCall(Vector=ParticleDirection(), Default=Constant3Vector() {constant=<rgba 1 0 0 0>}) {materialFunction=&SafeNormalize}.Result
   B=Constant3Vector() {constant=<rgba 0 0 1 0>}
  )
  n2 = If(A=Abs(None=@n1), B=Constant() {r=0.5}, "A > B"=Constant(), "A < B"=Constant() {r=1})
  n3 = MaterialFunctionCall(Float2=ConstantBiasScale(None=TextureCoordinate()) {bias=-0.5, desc="-.5, -1", scale=-1}) {materialFunction=&BreakOutFloat2Components}
  n4 = ParticleDirection()
  n5 = MaterialFunctionCall(Float2=ParticleSize()) {materialFunction=&BreakOutFloat2Components}
  n6 = Multiply(A=CrossProduct(A=@n4, B=Constant3Vector() {constant=<rgba 0 0 1 0>}), B=@n5.R)
  n7 = Multiply(A=@n5.G, B=@n4)
  n8 = Add(A=Add(A=Multiply(A=@n3.R, B=@n6), B=Multiply(A=@n3.G, B=@n7)), B=Multiply(A=WorldPosition().XYZ) {constB=-1})
  n9 = ParticlePositionWS()
  Add(
   A=Multiply(A=@n2, B=Add(A=@n8, B=@n9))
   B=MaterialFunctionCall!(World position) {materialFunction=/Game/WaterAndWind/MaterialFunctions/Fortnite_WindVectorParticleHelperFunctions}.Wind Particle WPO
  )
  n10 = LinearInterpolate(
   A=SphereMask(A=TextureCoordinate(), B=Constant() {r=0.5}) {attenuationRadius=0.5, hardnessPercent=0}
   B=TextureSample() {texture=/Game/WaterAndWind/Textures/WeaponFireVectorMask}.R
   Alpha=@n2
  )
  n11 = Add(A=ParticleRelativeTime(), B=@n10)
  n12 = Constant3Vector() {constant=<rgba 0 0 -1 0>}
  n13 = MaterialFunctionCall(Vector=Subtract(
    A=ParticlePositionWS()
    B=MaterialFunctionCall() {materialFunction=&Fortnite_WindVectorParticleHelperFunctions}.Pixel Shader Particle World Position - Adjusted For Offset
   )) {materialFunction=&SafeNormalize}
  MaterialFunctionCall(Emissive=Multiply(A=Sine(None=@n11) {period=2}, B=CrossProduct(A=@n12, B=@n13.Result))) {materialFunction=&MF_FX_EmissiveAdjustments}
  n14 = OneMinus(None=Clamp(None=@n1)) {desc=* opacity}
  n15 = ParticleColor()
  Multiply(A=Multiply(A=Multiply(A=@n14, B=@n14), B=Multiply(A=@n10, B=@n15.A)), B=VertexColor().A)
  OneMinus()
 ]
}
