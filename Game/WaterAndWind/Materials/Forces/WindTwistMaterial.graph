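# Unlit additive material that produces a twisting wind force vector masked to a soft disc, with negative emissive allowed so the vector keeps its sign.
# Centered, inverted UVs form an inward vector, rotated about Z by a sphere mask times Rotation Amount.
# Two samples of a noise texture on time-scrolled radial coordinates modulate that vector.
# The mask, Opacity and Intensity Tweak scale the result; particle color and wind helper nodes appear unconnected.
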Material /Game/WaterAndWind/Materials/Forces/WindTwistMaterial
graph {
 properties {
  bAllowNegativeEmissiveColor=true
  bUseTranslucencyVertexFog=false
  blendMode=BLEND_Additive
  pixelDepthOffsetMode=PDOM_Legacy
  shadingModel=MSM_Unlit
  usedShadingModels=Unlit
 }
 nodes [
  MaterialFunctionCall!(World position) {materialFunction=/Game/WaterAndWind/MaterialFunctions/Fortnite_WindVectorParticleHelperFunctions}
  MaterialFunctionCall() {materialFunction=&Fortnite_WindVectorParticleHelperFunctions}
  n1 = Multiply(A=Subtract(A=TextureCoordinate() {uTiling=2, vTiling=2})) {constB=-1}
  n2 = SphereMask(A=@n1, B=Constant()) {attenuationRadius=1, hardnessPercent=0}
  n3 = ScalarParameter() {defaultValue=0.685, parameterName=Rotation Amount}
  n4 = Constant()
  n5 = Time()
  n6 = Multiply(A=Constant2Vector() {g=1, r=-0.5}, B=ScalarParameter() {defaultValue=0.3945, parameterName=speed}) {constB=0.3}
  n7 = Frac(None=Add(
    A=MaterialFunctionCall() {materialFunction=&VectorToRadialValue}.Radial Coordinates
    B=Multiply(A=@n5, B=@n6) {constB=-1}
   ))
  n8 = Constant() {r=2}
  n9 = Multiply(A=Add(A=@n7) {constB=0.5}) {constB=2}
  n10 = Multiply(
   A=RotateAboutAxis(
    NormalizedRotationAxis=Constant3Vector() {constant=<rgba 0 0 1 0>}
    RotationAngle=Multiply(A=@n2, B=@n3)
    PivotPoint=Constant3Vector()
    Position=AppendVector(A=@n1, B=@n4)
   )
   B=Multiply(
    A=TextureSample!(UVs=@n7, Tex, Level=@n8, Apply View MipBias) {mipValueMode=TMVM_MipLevel, samplerType=SAMPLERTYPE_LinearColor, texture=/Game/Environments/Foliage/Textures/T_NoiseMask02_M}.R
    B=TextureSample(UVs=@n9, Level=@n8) {mipValueMode=TMVM_MipLevel, samplerType=SAMPLERTYPE_LinearColor, texture=&T_NoiseMask02_M}.R
   )
  )
  Multiply(
   A=Multiply(A=Multiply(A=@n2, B=@n10), B=ScalarParameter() {defaultValue=1, parameterName=Opacity})
   B=ScalarParameter() {defaultValue=10, parameterName=Intensity Tweak}
  )
  ParticleColor()
 ]
}
